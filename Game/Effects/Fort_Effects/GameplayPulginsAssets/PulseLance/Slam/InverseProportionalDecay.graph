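# InverseProportionalDecay is an FX-library material function in the PulseLance Slam effect folder that shapes a gradient into a flare or glow falloff.
# It scales the Emissive input down by a large constant, divides that by one minus the Gradient, then subtracts the scaled value so zero gradient yields zero.
# The saturated result follows a hyperbolic curve that climbs steeply to full brightness as Gradient nears one, with Emissive setting the glow's spread.
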MaterialFunction /Game/Effects/Fort_Effects/GameplayPulginsAssets/PulseLance/Slam/InverseProportionalDecay
graph {
 properties {
  description=Good for making flare / glow . I made this depend on Yue.Jia's summit sharing
  previewBlendMode=BLEND_Translucent
  libraryCategoriesText [FX]
 }
 nodes [
  n1 = Divide(A=FunctionInput() {inputName=Emissive, inputType=FunctionInput_Scalar, previewValue=<1 0 0 1>}) {constB=10000}
  n2 = Divide(A=@n1, B=OneMinus(None=FunctionInput() {inputName=Gradient, inputType=FunctionInput_Scalar}))
  FunctionOutput(None=Saturate(None=Subtract(A=@n2, B=@n1)))
 ]
}
