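# Material function in the FX library that draws a procedural four-pointed cross star from UVs, kept with the PulseLance slam effects.
# Swizzled UVs are scaled about center by a power curve, whose exponent blends from Fix_TipWidth to Tighten through a center sphere mask, forming concave arms.
# A DDX-based UV offset, scaled by an anti-flicker input, counters flickering tips.
# Outputs are a softness-controlled SmoothStep shape and its raw squared gradient.
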MaterialFunction /Game/Effects/Fort_Effects/GameplayPulginsAssets/PulseLance/Slam/ProceduralCrossStar
graph {
 properties {libraryCategoriesText [FX]}
 nodes [
  n1 = Constant()
  n2 = FunctionInput() {bUsePreviewValueAsDefault=true, inputName="Softness(0-1)", inputType=FunctionInput_Scalar, sortPriority=1, previewValue=<0.01 0 0 1>}
  n3 = Clamp(None=@n2)
  n4 = TextureCoordinate()
  n5 = DDX(Value=TextureCoordinate())
  n6 = FunctionInput() {
   bUsePreviewValueAsDefault=true
   inputName="Fix_AntiFlick(Increase This If The Tip Flicking)"
   inputType=FunctionInput_Scalar
   sortPriority=100
  }
  n7 = Multiply(A=MaterialFunctionCall(X=@n5, Y=@n5) {materialFunction=&MakeFloat2}.Result, B=@n6) {constB=10}
  n8 = Add(A=TextureCoordinate()) {constB=-0.5}
  n9 = LinearInterpolate(A=Multiply(A=@n7) {constB=-1}, B=@n7, Alpha=Ceil(None=Multiply(A=@n8) {constB=2}))
  n10 = TextureCoordinate()
  n11 = Add(A=Multiply(A=@n10, B=Constant2Vector() {g=2, r=2})) {constB=-1}
  n12 = FunctionInput() {bUsePreviewValueAsDefault=true, inputName=Fix_TipWidth, inputType=FunctionInput_Scalar, sortPriority=100, previewValue=<2 0 0 1>}
  n13 = FunctionInput() {bUsePreviewValueAsDefault=true, inputName=Tighten, inputType=FunctionInput_Scalar, previewValue=<1.8 0 0 1>}
  n14 = LinearInterpolate(A=@n12, B=@n13, Alpha=SphereMask(A=@n10, B=Constant() {r=0.5}) {attenuationRadius=0.5, hardnessPercent=0}) {constA=2}
  n15 = Power(Base=Saturate(None=Abs(None=@n11)), Exp=Divide(B=@n14))
  n16 = MaterialFunctionCall(
   UVs=MaterialFunctionCall(XY=Add(A=@n4, B=@n9)) {materialFunction=&Swizzle}.YX
   Texture Scale=Power(Base=OneMinus(None=@n15), Exp=@n14)
  ) {materialFunction=&ScaleUVsByCenter}
  n17 = Abs(None=Multiply(A=Add(A=@n16.UVs) {constB=-0.5}) {constB=2})
  n18 = MaterialFunctionCall(Float2=Saturate(None=OneMinus(None=@n17))) {materialFunction=&BreakOutFloat2Components}
  n19 = Saturate(None=Multiply(A=@n18.R, B=@n18.R))
  FunctionOutput(None=SmoothStep(Min=@n1, Max=Add(A=@n1, B=@n3), Value=@n19))
  FunctionOutput(None=@n19) {outputName=LinearGradient0-1, sortPriority=1}
 ]
}
