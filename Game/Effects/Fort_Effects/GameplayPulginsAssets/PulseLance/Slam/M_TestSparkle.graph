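# Translucent unlit test sparkle material in the PulseLance Slam effects folder.
# A ProceduralCrossStar function takes tightness and softness parameters to form a star shape.
# Separately, a panning-capable Perlin texture sample, scaled by a parameter, drives HueShift on a color parameter.
# The graph shows no link between the two branches or into material outputs, so it appears unfinished.
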Material /Game/Effects/Fort_Effects/GameplayPulginsAssets/PulseLance/Slam/M_TestSparkle
graph {
 properties {bAutomaticallySetUsageInEditor=false, blendMode=BLEND_Translucent, shadingModel=MSM_Unlit, usedShadingModels=Unlit}
 nodes [
  MaterialFunctionCall!(
   Tighten=ScalarParameter() {defaultValue=3.1622, parameterName=tr}
   "Softness(0-1)"=ScalarParameter() {defaultValue=0.6623, parameterName=W}
   "Fix_AntiFlick(Increase This If The Tip Flicking)"
   Fix_TipWidth
  ) {materialFunction=&/ProceduralCrossStar}
  n1 = TextureObjectParameter() {
   group=Hue
   parameterName=Hue_Texture
   samplerType=SAMPLERTYPE_LinearGrayscale
   sortPriority=0
   texture=&T_SimplePerlin_Greyscale
  }
  n2 = VectorParameter() {group=Hue, parameterName=Hue_UVControl, channelNames {a=VPan, b=UPan, g=VScale, r=UScale}, defaultValue=<rgba 1 1 0.5 0.5>}
  MaterialFunctionCall(
   Hue Shift Percentage=Multiply(
    A=MaterialFunctionCall(Texture=@n1, UVControl_RGB=@n2.RGB, UVControl_A=@n2.RGB) {materialFunction=&FXCommon_TextureSample}.R
    B=ScalarParameter() {defaultValue=0.4032, parameterName=H}
   )
   Texture=VectorParameter() {parameterName=C, defaultValue=<rgba 0.5478 5 3.5365 1>}.RGB
  ) {materialFunction=&HueShift}
 ]
}
